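# Animation Blueprint for the ArcticIceBlue male head that eases neck rotation when the character wears a backbling.
# Each update, if a backbling is detected, it maps negative yaw of neck_01's ref-pose delta relative to spine_05 to a clamped, scaled-down weight; otherwise the weight is zero.
# That weight drives a per-bone layered blend from neck_01, pulling the input pose partly toward the unconnected layer's default pose.
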AnimBlueprint /BRCosmetics/Characters/Player/Male/Medium/Heads/M_MED_ArcticIceBlue_Head/Meshes/M_MED_ArcticIceBlue_Head_AnimBP_HeadRotation
target_skeleton=/Game/Characters/Player/Common/Fortnite_Base_Head/Fortnite_Base_Head_Export_Skeleton
graphs [
 AnimGraph {nodes [
   n1 = Layered blend per bone!(
    BasePose=Input Pose()
    BlendPoses_0
    BlendPoses_1:Pose Link Structure
    BlendWeights_0=Get neckRotateBack() -> (neckRotateBack:double)
    BlendWeights_1:float=1
   ) -> (Pose) {
    blendPoses [{linkId=-1, sourceLinkId=-1}, {linkId=-1, sourceLinkId=-1}]
    blendWeights [1,1]
    layerSetup [{branchFilters [{blendDepth=1, boneName=neck_01}]}, {branchFilters []}]
   }
   Output Pose(Result=@n1) {layerGroup=None}
  ]}
 EventGraph {nodes [
   Event BlueprintUpdateAnimation()
   > Set Is Wearing Backbling(Is Wearing Backbling=CheckForBackbling())
   > n1 = Branch(Condition=Get Is Wearing Backbling())
   n2 = GetOwnerSkeletalMeshComponent()
   n3 = BreakRotator(InRot=BreakTransform(InTransform=GetDeltaTransformFromRefPose(self=@n2, BoneName=neck_01, BaseName=spine_05)).Rotation)
   > Set neckRotateBack(neckRotateBack:double=float * float(A=MapRangeClamped(Value=@n3.Yaw, InRangeA=0, InRangeB=-20), B=0.3) {is=title K2Node_PromotableOperator}) -> (Output_Get:double)
   Set neckRotateBack(execute=@n1.else, neckRotateBack=0)
  ]}
]
variables [
 head rot {type=Rotator}
 HeadRotPercentage {type=number, default=1}
 Is Wearing Backbling {type=boolean}
 neckRotateBack {type=number}
]
